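# Niagara system for a Mathematic emote that pops a one-shot blue symbol sprite, apparently the multiplication sign by its name, with glow and sparkles.
# The symbol is one sub-image particle drawn by two sprite renderers, with a brightness flash and a curve-driven pop, pulse and shrink.
# FlareBurst adds four faint, drifting blue-purple flares, and StartsBurst a few small, bright cyan sparks with drag, fading out.
# Camera offsets layer the symbol ahead of both bursts.
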NiagaraSystem /BRCosmetics/Athena/Effects/Particles/Emotes/NS_Mathematic_Emotes/NS_Emotes_Mathematic_03_Mult
dependencies {usedDataInterfaces [&NiagaraDataInterfaceCurve], usedDynamicInputs [], usedModules [&SystemState]}
emitters {
 FN_Emotes_Mathematic_Mult {
  emitterUpdateScript [
   &EmitterScalability(
    Module.bApplySpawnCountScale=false
    Module.bApplySpawnCountScaleByDistanceFraction=false
    Module.bUseMaxDistance=false
    Module.bUseMinDistance=false
   ) {enabled=false}
   &EmitterLifeCycle(CompleteOnInactive=true, MaxLoopCount=1, NextLoopDuration=10) {enabled=false}
   &EmitterState()
   &SpawnRate(SpawnRate=20, Use Spawn Probability=false) {enabled=false}
   &SpawnBurst_Instantaneous(Spawn Count=1, Use Spawn Probability=false)
  ]
  particleSpawnScript [
   /Niagara/Modules/Spawn/Initialization/InitializeParticle(
    Color=<rgba 0 1.25 5 1>
    Lifetime=0.7
    Module.Write Color=true
    Module.Write Lifetime=true
    Module.Write Mass=true
    Module.Write Material Random=true
    Module.Write Position=true
    Module.Write Scale=false
    Module.Write SpriteRotation=false
    Module.Write SpriteSize=true
    Position=&SimulationPosition()
    Sprite Size=&Vector2DFromFloat(Value=65)
   )
   /SetVariables(Particles.SubImageIndex=2) {isSetParameters=true}
   /SetVariables(Particles.CameraOffset=13) {isSetParameters=true}
  ]
  particleUpdateScript [
   &ParticleState()
   &ScaleColor(
    Scale RGB=&VectorFromFloat(Value=&FloatFromCurve(FloatCurve=DataInterface() {propertyValues {Curve {keys [<at 0 0>,<at 0.1 10>,<at 0.2 1>,<at 1 1>]}}}))
    ScaleA=true
    ScaleRGB=true
   )
   &ScaleSpriteSize(Scale Factor=&Vector2DFromFloat(Value=&FloatFromCurve(FloatCurve=DataInterface() {propertyValues {Curve {keys [
          {interpMode=RCIM_Cubic, leaveTangent=10.1561, tangentMode=RCTM_Break, time=0, value=0}
          {arriveTangent=0.0268, interpMode=RCIM_Cubic, leaveTangent=0.022, tangentMode=RCTM_Break, time=0.25, value=1}
          {interpMode=RCIM_Cubic, tangentMode=RCTM_Break, time=0.5, value=0.8}
          {arriveTangent=0.0134, interpMode=RCIM_Cubic, leaveTangent=0.0541, tangentMode=RCTM_Break, time=0.75, value=1}
          {arriveTangent=-6.3327, interpMode=RCIM_Cubic, tangentMode=RCTM_Break, time=1, value=0}
         ]}}})))
   &SystemLocation(MaskSpawnGroup=false)
  ]
  rendererClasses [&NiagaraSpriteRendererProperties]
  renderers [
   {rendererClass=&NiagaraSpriteRendererProperties}
   {rendererClass=&NiagaraSpriteRendererProperties, rendererIndex=1}
  ]
 }
 FlareBurst {
  emitterUpdateScript [
   &EmitterScalability(
    Module.bApplySpawnCountScale=false
    Module.bApplySpawnCountScaleByDistanceFraction=false
    Module.bUseMaxDistance=false
    Module.bUseMinDistance=false
   ) {enabled=false}
   &EmitterLifeCycle(MaxLoopCount=1, NextLoopDuration=0.1) {enabled=false}
   &EmitterState()
   &SpawnBurst_Instantaneous(Spawn Count=4, Use Spawn Probability=false)
   &SpawnRate(SpawnRate=15, Use Spawn Probability=false) {enabled=false}
  ]
  particleSpawnScript [
   &InitializeParticle(
    Color=&UniformRangedLinearColor(Minimum=<rgba 0.0668 0.4843 1 0.05>, Maximum=<rgba 0.3472 0.1667 1 0.04>)
    Lifetime=&UniformRangedFloat(Minimum=0.8, Maximum=0.3)
    Module.Write Color=true
    Module.Write Lifetime=true
    Module.Write Mass=true
    Module.Write Material Random=true
    Module.Write Position=true
    Module.Write Scale=false
    Module.Write SpriteRotation=true
    Module.Write SpriteSize=true
    Position=&SimulationPosition()
    Sprite Rotation=&UniformAOrBFloat(A=90, B=-90)
    Sprite Size=&Vector2DFromFloat(Value=&Multiply_Float(A=1.8, B=&UniformRangedFloat(Minimum=15, Maximum=40)))
   )
   &CameraOffset(Camera Offset Amount=10)
   /Niagara/Modules/Spawn/Location/SphereLocation(MaskSpawnGroup=false, Module.Override Randomness=false, Offset=<0 0 10>, Sphere Radius=25, Surface Only=false)
   &AddVelocity(Velocity=&UniformRangedVector(Minimum=<0 0 30>, Maximum=<1 1 10>))
  ]
  particleUpdateScript [
   &ParticleState()
   &SolveForcesAndVelocity(Clamp Velocity=false, Limit Acceleration=false)
   &ScaleColor(
    Scale Alpha=&FloatFromCurve(FloatCurve=DataInterface() {propertyValues {Curve {keys [
         <at 0 0 RCIM_Cubic>
         {arriveTangent=-0.0426, interpMode=RCIM_Cubic, leaveTangent=-0.0426, tangentMode=RCTM_User, time=0.3771, value=1.0137}
         {arriveTangent=0.0121, interpMode=RCIM_Cubic, leaveTangent=0.0121, time=0.5746, value=0.6676}
         {arriveTangent=-1.5692, interpMode=RCIM_Cubic, leaveTangent=-1.5692, time=0.8024, value=1.0189}
         <at 1 0>
        ]}}})
    ScaleA=true
    ScaleRGB=true
   )
   &ScaleSpriteSize(Scale Factor=&Vector2DFromFloat(Value=&FloatFromCurve(FloatCurve=DataInterface() {propertyValues {Curve {keys [<at 0.0121 0.5565>,<at 0.491 0.8934>]}}})))
  ]
  rendererClasses [&NiagaraSpriteRendererProperties]
  renderers [{rendererClass=&NiagaraSpriteRendererProperties}]
 }
 StartsBurst {
  emitterUpdateScript [
   &EmitterScalability(
    Module.bApplySpawnCountScale=false
    Module.bApplySpawnCountScaleByDistanceFraction=false
    Module.bUseMaxDistance=false
    Module.bUseMinDistance=false
   ) {enabled=false}
   &EmitterLifeCycle(MaxLoopCount=1, NextLoopDuration=0.1) {enabled=false}
   &EmitterState()
   &SpawnBurst_Instantaneous(Spawn Count=&UniformRangedInt(Minimum=6, Maximum=4), Use Spawn Probability=false)
   &SpawnRate(SpawnRate=3, Use Spawn Probability=false) {enabled=false}
  ]
  particleSpawnScript [
   &InitializeParticle(
    Color=&UniformRangedLinearColor(Minimum=<rgba 4.974 15.7608 20 1>, Maximum=<rgba 4.974 15.7608 20 1>)
    Lifetime=&UniformRangedFloat(Minimum=0.5, Maximum=0.8)
    Module.Write Color=true
    Module.Write Lifetime=true
    Module.Write Mass=true
    Module.Write Material Random=true
    Module.Write Position=true
    Module.Write Scale=false
    Module.Write SpriteRotation=true
    Module.Write SpriteSize=true
    Position=&SimulationPosition()
    Sprite Rotation=&UniformRangedFloat(Minimum=0.02, Maximum=-0.02)
    Sprite Size=&UniformRangedVector2D(Minimum=<3 10>, Maximum=<5 4>)
   )
   &CameraOffset(Camera Offset Amount=5)
   &AddVelocity(Velocity=&UniformRangedVector(Minimum=<0 0 50>, Maximum=<0 0 0>))
   &SphereLocation(MaskSpawnGroup=false, Module.Override Randomness=false, Offset=<0 0 25>, Sphere Radius=30, Surface Only=false)
  ]
  particleUpdateScript [
   &ParticleState()
   &Drag(Drag=2, Use Linear Drag=true, Use Rotational Drag=false)
   &SolveForcesAndVelocity(Clamp Velocity=false, Limit Acceleration=false)
   &ScaleColor(
    Scale Alpha=&FloatFromCurve(FloatCurve=DataInterface() {propertyValues {Curve {keys [{arriveTangent=0.0145, interpMode=RCIM_Cubic, leaveTangent=0.0145, tangentMode=RCTM_User, time=0.2, value=1}, <at 1 0>]}}})
    Scale RGB=&VectorFromFloat(Value=&FloatFromCurve(FloatCurve=DataInterface() {propertyValues {Curve {keys [<at 0 50>,<at 0.1 1>]}}}))
    ScaleA=true
    ScaleRGB=true
   )
   &ScaleSpriteSize(Scale Factor=&Vector2DFromFloat(Value=&FloatFromCurve(
      FloatCurve=DataInterface() {propertyValues {Curve {keys [<at 0.0121 0.5565>,<at 0.491 0.8934>,<at 1.0011 0.002>]}}}
      Scale Curve=&UniformRangedFloat(Minimum=0.7, Maximum=1.5)
     )))
  ]
  rendererClasses [&NiagaraSpriteRendererProperties]
  renderers [{rendererClass=&NiagaraSpriteRendererProperties}]
 }
}
summary {
 systemName=NS_Emotes_Mathematic_03_Mult
 emitters [
  {emitterName=FN_Emotes_Mathematic_Mult, rendererClasses [&NiagaraSpriteRendererProperties]}
  {emitterName=FlareBurst, rendererClasses [&NiagaraSpriteRendererProperties]}
  {emitterName=StartsBurst, rendererClasses [&NiagaraSpriteRendererProperties]}
 ]
 userVariables [User.InitialOffset {type=Vector3f, description="", defaultValue {struct=Vector3f, value=<0 0 0>}}]
}
