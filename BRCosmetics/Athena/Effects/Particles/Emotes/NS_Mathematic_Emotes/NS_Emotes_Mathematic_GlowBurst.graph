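# Niagara system for the Mathematic emote that fires a short glow burst of blue-to-violet sprites along the player's forearms.
# Its single FlareBrust emitter spawns an instant burst, with spawn groups splitting particles between the lowerarm_l and lowerarm_r bones.
# Every update re-samples the bone and applies a random, side-mirrored offset rotated by it, so sprites follow the arms.
# Curves shape alpha and size over each particle's brief life, with boosted RGB for glow.
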NiagaraSystem /BRCosmetics/Athena/Effects/Particles/Emotes/NS_Mathematic_Emotes/NS_Emotes_Mathematic_GlowBurst
dependencies {usedDataInterfaces [&NiagaraDataInterfaceCurve], usedDynamicInputs [], usedModules [&SystemState]}
emitters {FlareBrust {
  emitterUpdateScript [
   &EmitterScalability(
    Module.bApplySpawnCountScale=false
    Module.bApplySpawnCountScaleByDistanceFraction=false
    Module.bUseMaxDistance=false
    Module.bUseMinDistance=false
   ) {enabled=false}
   &EmitterLifeCycle(MaxLoopCount=1, NextLoopDuration=0.1) {enabled=false}
   &EmitterState()
   &SpawnBurst_Instantaneous(Spawn Count=10, Use Spawn Probability=false)
   &SpawnRate(SpawnRate=15, Use Spawn Probability=false) {enabled=false}
  ]
  particleSpawnScript [
   /Niagara/Modules/Spawn/Initialization/InitializeParticle(
    Color=&UniformRangedLinearColor(Minimum=<rgba 0 0.4474 2 0.4>, Maximum=<rgba 0.2167 0 2 0.2>)
    Lifetime=&UniformRangedFloat(Minimum=0.4, Maximum=0.6)
    Module.Write Color=true
    Module.Write Lifetime=true
    Module.Write Mass=true
    Module.Write Material Random=true
    Module.Write Position=true
    Module.Write Scale=false
    Module.Write SpriteRotation=true
    Module.Write SpriteSize=true
    Position=&SimulationPosition()
    Sprite Rotation=&UniformRangedFloat(Minimum=0, Maximum=360)
    Sprite Size=&Vector2DFromFloat(Value=&UniformRangedFloat(Minimum=25, Maximum=25))
   )
   /SetVariables(Particles.Distribution_L:Vector3f=&UniformRangedVector(Minimum=<0 0 0>, Maximum=<30 0 0>)) {isSetParameters=true}
   /SetVariables(Particles.Distribution_R:Vector3f=&UniformRangedVector(Minimum=<0 0 0>, Maximum=<-30 0 0>)) {isSetParameters=true}
   &SystemLocation(MaskSpawnGroup=false, Offset=&UniformRangedVector(Minimum=<16 0 0>, Maximum=<-16 0 0>)) {enabled=false}
   &CameraOffset(Camera Offset Amount=5)
   /Game/Effects/Niagara/Modules/Groups/FN_SetSpawnGroup_v1/FN_SetSpawnGroup(Emitter Bone Count=2)
  ]
  particleUpdateScript [
   &ParticleState()
   &SolveForcesAndVelocity(Clamp Velocity=false, Limit Acceleration=false)
   &ScaleColor(
    Scale Alpha=&FloatFromCurve(
     FloatCurve=DataInterface() {propertyValues {Curve {keys [
         {interpMode=RCIM_Cubic, leaveTangent=0.6032, tangentMode=RCTM_Break, time=0, value=0}
         <at 0.2964 1.3399>
         {arriveTangent=0.0145, interpMode=RCIM_Cubic, leaveTangent=0.0145, tangentMode=RCTM_User, time=0.6, value=1}
         <at 1 0>
        ]}}}
     Scale Curve=0.5
    )
    Scale RGB=<5 5 5>
    ScaleA=true
    ScaleRGB=true
   )
   &ScaleSpriteSize(Scale Factor=&Vector2DFromFloat(Value=&FloatFromCurve(FloatCurve=DataInterface() {propertyValues {Curve {keys [<at 0.0121 0.5565>,<at 0.7048 0.9115>,<at 1.0011 0.002>]}}})))
   &SampleSkeletalMeshSkeleton(
    Mesh=DataInterface() {propertyValues {
      DefaultMesh=null
      ExcludeBoneName=root
      OptionalComponentClass=null
      PreviewMesh=/Game/Characters/Player/Male/Male_Avg_Base/Fortnite_M_Avg_Player
      SkinningMode=SkinOnTheFly
      SoftSourceActor=null
      SourceMode=Default
      UvSetIndex=0
      WholeMeshLOD=-1
      bExcludeBone=false
      bReadDeformedGeometry=true
      bRequireCurrentFrameData=true
      ComponentTags []
      FilteredBones [lowerarm_l,lowerarm_r]
      FilteredSockets []
      MeshUserParameter {parameter {name=None, typeDefHandle {registeredTypeIndex=1}, varData []}}
      SamplingRegions []
     }}
    Specific Bone Index:NiagaraInt32=$Particles.SpawnGroup
   )
   /Niagara/Modules/Spawn/Location/SkeletalMeshSkeletonLocation(
    Module.Inherit Velocity=false
    Module.Use Sampled Bone as Ribbon ID=true
    Sampled Position Offset:Vector3f=&RotateVectorByQuaternion(
     Vector To Rotate=/Game/Effects/Niagara/DynamicInputs/Water/FN_PickVectorFromSpawnGroupCutoff(
      Under Cutoff Vector:Vector3f=$Particles.Distribution_L
      Over Cutoff Vector:Vector3f=$Particles.Distribution_R
      SpawnGroup:NiagaraInt32=$Particles.SpawnGroup
      Cutoff:NiagaraInt32=1
     )
     Quaternion=$Particles.SampleSkeletalMeshSkeleton.SampledRotation
    )
   )
  ]
  rendererClasses [&NiagaraSpriteRendererProperties]
  renderers [{rendererClass=&NiagaraSpriteRendererProperties}]
 }}
summary {
 systemName=NS_Emotes_Mathematic_GlowBurst
 emitters [{emitterName=FlareBrust, rendererClasses [&NiagaraSpriteRendererProperties]}]
 userVariables [
  User.bIsFrontEndPreview {type=NiagaraBool}
  User.FEVec {type=Vector3f, description="", defaultValue {struct=Vector3f, value=<0 0 1000>}}
  User.PlayerSpeed {type=NiagaraFloat, defaultValue=0, description=""}
  User.RibbonIDOffset {type=NiagaraInt32}
  User.SkeletalMesh {
   type=NiagaraDataInterfaceSkeletalMesh
   defaultValue {
    struct=NiagaraExt_VariableValue_DataInterface
    value {dataInterface=&self:NiagaraDataInterfaceSkeletalMesh_0}
   }
  }
  User.SpawnRate {type=NiagaraFloat, defaultValue=360, description=""}
 ]
}
